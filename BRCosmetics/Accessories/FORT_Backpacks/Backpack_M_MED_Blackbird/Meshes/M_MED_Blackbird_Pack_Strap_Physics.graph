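# Physics asset for the dangling strap on the Blackbird backpack cosmetic in Battle Royale.
# A strap chain of small capsules hangs from the pack bone through a master link, with tight angular limits at that anchor and wider swing on one axis down the chain.
# A second chain of three large spine capsules links to the pack with free rotation, likely standing in for the wearer's back as collision.
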PhysicsAsset /BRCosmetics/Accessories/FORT_Backpacks/Backpack_M_MED_Blackbird/Meshes/M_MED_Blackbird_Pack_Strap_Physics
bodies {
 dyn_spine_02 [{shapeType=Capsule, length=15, radius=12, shapeName=dyn_spine_02_capsule}]
 dyn_spine_03 [{shapeType=Capsule, length=15, radius=12, shapeName=dyn_spine_03_capsule}]
 dyn_spine_04 [{shapeType=Capsule, length=15, radius=12, shapeName=dyn_spine_04_capsule}]
 dyn_strap_01 [{shapeType=Capsule, length=1.5, radius=1.5, shapeName=dyn_strap_01_capsule, center=<2.2615 0 0>, rotation=<rot 0 0 90>}]
 dyn_strap_01_MSTR [{shapeType=Capsule, length=0.505, radius=0.505, shapeName=dyn_strap_01_MSTR_capsule}]
 dyn_strap_02 [{shapeType=Capsule, length=1.5, radius=1.5, shapeName=dyn_strap_02_capsule, center=<2.425 0 0>, rotation=<rot 0 0 90>}]
 dyn_strap_03 [{shapeType=Capsule, length=2.5, radius=2, shapeName=dyn_strap_03_capsule, center=<2 0 0>, rotation=<rot 0 0 90>}]
 pack [{shapeType=Capsule, length=0.505, radius=0.505, shapeName=pack_capsule, center=<0.0008 -8.5653 -0.5769>}]
}
constraints [
 {bone1Name=dyn_strap_01_MSTR, bone2Name=pack, swing1LimitDegrees=5, swing2LimitDegrees=5, twistLimitDegrees=5}
 {bone1Name=dyn_strap_01, bone2Name=dyn_strap_01_MSTR, swing1LimitDegrees=15, swing2LimitDegrees=45, twistLimitDegrees=15}
 {bone1Name=dyn_strap_02, bone2Name=dyn_strap_01, swing1LimitDegrees=15, swing2LimitDegrees=60, twistLimitDegrees=15}
 {bone1Name=dyn_strap_03, bone2Name=dyn_strap_02, swing1LimitDegrees=15, swing2LimitDegrees=60, twistLimitDegrees=15}
 {
  bone1Name=dyn_spine_02
  bone2Name=pack
  swing1LimitDegrees=45
  swing1Motion=Free
  swing2LimitDegrees=45
  swing2Motion=Free
  twistLimitDegrees=45
  twistMotion=Free
 }
 {
  bone1Name=dyn_spine_03
  bone2Name=dyn_spine_02
  swing1LimitDegrees=45
  swing1Motion=Free
  swing2LimitDegrees=45
  swing2Motion=Free
  twistLimitDegrees=45
  twistMotion=Free
 }
 {
  bone1Name=dyn_spine_04
  bone2Name=dyn_spine_03
  swing1LimitDegrees=45
  swing1Motion=Free
  swing2LimitDegrees=45
  swing2Motion=Free
  twistLimitDegrees=45
  twistMotion=Free
 }
]
